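; BP_DFAffector is an actor that holds cube, sphere and cylinder DFAffector instanced static mesh components.
; Custom events add an instance at a zero location to each shape's component, and Toggle Visibility uses a FlipFlop to set visibility on the cylinder component only.
; The construction script sets a documentation link, then filters the three built-in components from all BPC_ISM_DFAffector components and destroys what remains.
; BeginPlay, overlap and Tick events are empty.
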(event EventBeginPlay)

(event Collision|EventActorBeginOverlap
  (params (OtherActor :type Actor%20Object%20Reference)))

(event EventTick
  (params (DeltaSeconds :type Float%20%28single-precision%29)))

(event Custom|Toggle%20Visibility
  (bind _isa (Utilities|FlowControl|FlipFlop)
    (:A
      _)
    (:B
      _))
  (Rendering|SetVisibility (Variables|Default|GetBPC_ISM_DFAffector_Cylinder) (not _isa) false))

(event Custom|Add%20Cube
  (Components|InstancedStaticMesh|AddInstance (Variables|Default|GetBPC_ISM_DFAffector_Cube) (Math|Transform|MakeTransform "0, 0, 0") false))

(event Custom|Add%20Sphere
  (Components|InstancedStaticMesh|AddInstance (Variables|Default|GetBPC_ISM_DFAffector_Sphere) (Math|Transform|MakeTransform "0, 0, 0") false))

(event Custom|Add%20Cylinder
  (Components|InstancedStaticMesh|AddInstance (Variables|Default|GetBPC_ISM_DFAffector_Cylinder) (Math|Transform|MakeTransform "0, 0, 0") false))

(fn ConstructionScript
  (bind _newlocalvar (LocalVariables|GetNewLocalVar))
  (Utilities|FlowControl|Sequence
    (:then_0
      (Variables|Info|SetDocumentation "https://confluence-epicgames.atlassian.net/wiki/x/PglkCg"))
    (:then_1
      (bind _output_get (LocalVariables|SetNewLocalVar (Actor|GetComponentsByClass "/Game/Environments/TechArt/DFAffector/BPC_ISM_DFAffector.BPC_ISM_DFAffector_C")))
      (Utilities|Array|RemoveItem _output_get (Variables|Default|GetBPC_ISM_DFAffector_Sphere))
      (Utilities|Array|RemoveItem _output_get (Variables|Default|GetBPC_ISM_DFAffector_Cube))
      (Utilities|Array|RemoveItem _output_get (Variables|Default|GetBPC_ISM_DFAffector_Cylinder))
      (if (Utilities|Array|IsNotEmpty _newlocalvar)
        (Components|DestroyComponent _newlocalvar)))))
